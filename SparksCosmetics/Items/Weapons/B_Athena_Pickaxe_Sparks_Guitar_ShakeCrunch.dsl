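; Sparks guitar pickaxe cosmetic Blueprint that cycles its collage outline color.
; BeginPlay calls the parent, starts a looping timer with a randomized initial delay, and calls the toon frontend LUT replace helper for pickaxes.
; Each tick of the timer picks a random CollageColor entry and writes it to the weapon mesh materials' Collage_Color parameter, with brightness scaled according to whether the item is displayed in a frontend context.
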(event EventBeginPlay
  (|CallParentFunction)
  ;[1 0.8396 0.3695] Collage Outline color change
  (Utilities|Time|SetTimerbyEvent (delegate Custom|ChangeCollageColor) 3.3 true false 6.0 1.3)
  (Utilities|ToonFrontendLUTReplacePickaxe))

(event Custom|ChangeCollageColor
  ;[1 0.8396 0.3695] Collage Outline color change
  (CallFunction|SetRandomCollageColors))

(fn SetRandomCollageColors
  (bind _element (ArrayMacros|GetRandomArrayElement (Variables|Default|GetCollageColor))
    (:%20%20
      (Rendering|Material|SetColorParameterValueonMaterials (Variables|Visual|GetWeaponMesh) "Collage_Color" (* (Math|Conversions|ToLinearColor(Color) _element) (select (Reactive|Frontend|IsCosmeticObjectDisplayedinFrontendContext 0) "(R=0.700000,G=0.700000,B=0.700000,A=1.000000)" "(R=1.000000,G=1.000000,B=1.000000,A=1.000000)"))))))

(fn ConstructionScript)
