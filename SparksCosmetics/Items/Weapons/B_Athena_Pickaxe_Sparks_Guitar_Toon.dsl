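; Toon-style Sparks guitar pickaxe weapon Blueprint that corrects its material LUT when displayed in the frontend.
; After cosmetic initialization, it loops over weapon mesh materials, reusing dynamic instances or creating and assigning new ones.
; Materials whose PBR_Shading is not 1 get the frontend inverse LUT volume texture and related scalar overrides.
; An in-graph note marks this as a stopgap until an automatic LUT tool exists.
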(event EventHandlePostCosmeticInitialization
  (bind _current_mid (Variables|Default|GetCurrent%20MID))
  ;[1 1 1] Front End LUT, Remove after creating auto LUT tool
  (if (Reactive|Frontend|IsCosmeticObjectDisplayedinFrontendContext self)
    (for (_array_element _array_index) (Rendering|Material|GetMaterials (Variables|Visual|GetWeaponMesh))
      (bind _asmaterial_instance_dynamic (Utilities|Casting|CastToMaterialInstanceDynamic _array_element)
        (:then
          (Variables|Default|SetCurrent%20MID _asmaterial_instance_dynamic))
        (:CastFailed
          (bind _returnvalue (Rendering|Material|CreateDynamicMaterialInstance _array_element "None" "None" :!class /Script/Engine.KismetMaterialLibrary))
          (bind _output_get (Variables|Default|SetCurrent%20MID _returnvalue))
          (Rendering|Material|SetMaterial (Variables|Visual|GetWeaponMesh) _array_index _output_get)))
      (bind _returnvalue_2 (Rendering|Material|GetScalarParameterValue _current_mid "PBR_Shading" :!class /Script/Engine.MaterialInstanceDynamic))
      (if (not (== _returnvalue_2 1.0))
        (Rendering|Material|SetTextureParameterValue _current_mid "LUT Volume Texture" "/Game/Packages/Fortress_SharedMaterials/Anime_Master_Material/VT_InverseLUT_FE.VT_InverseLUT_FE")
        (Rendering|Material|SetScalarParameterValue _current_mid "LUT Multiplier" 1.0 :!class /Script/Engine.MaterialInstanceDynamic)
        (Rendering|Material|SetScalarParameterValue _current_mid "Mobile LUT Brightness" 1.0 :!class /Script/Engine.MaterialInstanceDynamic)
        (Rendering|Material|SetScalarParameterValue _current_mid "use 33 Resolution" 0.0 :!class /Script/Engine.MaterialInstanceDynamic)))))

(fn ConstructionScript)
