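; Audio-reactive toon guitar pickaxe cosmetic that fakes an audio reaction when displayed in the frontend or as a preview.
; On weapon cosmetic init in those contexts, a timeline drives the EnableFrontendFakeAudioLoop parameter on the weapon mesh materials and idle VFX.
; The timeline plays forward, waits an on-length delay, then reverses, using a BeginLoop flag to select the following off-length delay.
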(event Custom|OnFakeLoopTimelineUpdate
  (params (TransitionValue :type Float%20%28single-precision%29))
  (Rendering|Material|SetScalarParameterValueonMaterials (Variables|Visual|GetWeaponMesh) "EnableFrontendFakeAudioLoop" TransitionValue) ; Enable Fake Audio in Frontend for Materials
  (Effects|Components|ParticleSystem|SetFloatParameter (CallFunction|GetIdleVFXComponent) "EnableFrontendFakeAudioLoop" TransitionValue))

(event EventOnInitWeaponCosmetics
  (|CallParentFunction)
  (if (or (Reactive|Frontend|IsCosmeticObjectDisplayedinFrontendContext 0) (Reactive|Frontend|IsCosmeticObjectDisplayedasPreview 0))
    (Variables|Default|SetBeginLoop true)
    (bind (_direction _in) (Timeline|Timeline PlayFromStart)
      (:Update
        (CallFunction|OnFakeLoopTimelineUpdate _in))
      (:Finished
        (if (Variables|Default|GetBeginLoop)
          (Variables|Default|SetBeginLoop false)
          (Utilities|FlowControl|Delay (Variables|Default|GetFakeAmplitudeOnLength))
          (Timeline|Timeline ReverseFromEnd)
          (else
            (Utilities|FlowControl|Delay (Variables|Default|GetFakeAmplitudeOffLength))))))))

(fn ConstructionScript)
