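; Keytar pickaxe weapon Blueprint for the Sparks cosmetics, customizing idle and impact VFX.
; On instigator set and on equip, it stores the pawn's mesh and binds it to the idle Niagara component's User.InstigatorMesh, with a short delay if that component is invalid.
; Left and right swing events set a flag that orients impact effects along the hit normal, which also receive a surface-type index.
; Construction disables idle particle auto-activation.
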(event FortWeapon|EventOnInstigatorSet
  (bind _returnvalue (CallFunction|GetIdleVFXComponent))
  (|CallParentFunction)
  (bind _asfort_pawn (Utilities|Casting|CastToFortPawn (Game|GetInstigator)))
  (Variables|Default|SetInstigatorSkeletalMesh (Variables|/Script/Engine.Character|GetMesh _asfort_pawn))
  (Utilities|IsValid _returnvalue
    (:Is%20Valid
      _)
    (:Is%20Not%20Valid
      (Utilities|FlowControl|Delay 0.2)))
  (bind _asniagara_particle_system_component (Utilities|Casting|CastToNiagaraComponent _returnvalue)
    (:then
      (Niagara|SetNiagaraSkeletalMeshComponent _asniagara_particle_system_component "User.InstigatorMesh" (Variables|Default|GetInstigatorSkeletalMesh)))
    (:CastFailed
      (Development|PrintString "Niagara Cast Failed"))))

(event FortWeapon|EventOnEquippedAndAttached
  (bind _returnvalue (CallFunction|GetIdleVFXComponent))
  (|CallParentFunction)
  (bind _asfort_pawn (Utilities|Casting|CastToFortPawn (Game|GetInstigator)))
  (Variables|Default|SetInstigatorSkeletalMesh (Variables|/Script/Engine.Character|GetMesh _asfort_pawn))
  (Utilities|IsValid _returnvalue
    (:Is%20Valid
      _)
    (:Is%20Not%20Valid
      (Utilities|FlowControl|Delay 0.2)))
  (bind _asniagara_particle_system_component (Utilities|Casting|CastToNiagaraComponent _returnvalue)
    (:then
      (Niagara|SetNiagaraSkeletalMeshComponent _asniagara_particle_system_component "User.InstigatorMesh" (Variables|Default|GetInstigatorSkeletalMesh)))
    (:CastFailed
      (Development|PrintString "Niagara Cast Failed"))))

(event FortWeapon|EventOnLocalUnEquip
  (|CallParentFunction))

(event FortWeapon|EventOnUnEquippedAndDetached
  (|CallParentFunction))

(event FortWeapon|EventOnPlayImpactFX
  (params (HitResult :type Hit%20Result%20Structure%20%28by%20ref%29) (ImpactPhysicalSurface :type EPhysicalSurface%20Enum :default "SurfaceType_Default") (SpawnedPSC :type FXSystem%20Component%20Object%20Reference))
  (bind (_bblockinghit _binitialoverlap _time _distance _location _impactpoint _normal) (Collision|BreakHitResult HitResult))
  (Transformation|SetWorldRotation SpawnedPSC (Math|Rotator|CombineRotators (Math|Rotator|MakeRotfromX _normal) (select (Variables|Default|GetRight%20?) (Variables|Default|GetRight%20Rotation) (Variables|Default|GetLeft%20Rotation))) false false)
  (Effects|Components|ParticleSystem|SetIntParameter SpawnedPSC "PhysicalSurface" (Utilities|Select :Index ImpactPhysicalSurface :SurfaceType_Default 0 :SurfaceType1 1 :SurfaceType2 2 :SurfaceType3 3 :SurfaceType4 4 :SurfaceType5 5 :SurfaceType6 6 :SurfaceType7 7 :SurfaceType8 8 :SurfaceType9 9 :SurfaceType10 10 :SurfaceType11 11 :SurfaceType12 12 :SurfaceType13 13 :SurfaceType14 14 :SurfaceType15 15 :SurfaceType16 16 :SurfaceType17 17 :SurfaceType18 18 :SurfaceType19 19 :SurfaceType20 20 :SurfaceType21 21 :SurfaceType22 22 :SurfaceType23 23 :SurfaceType24 24 :SurfaceType25 25 :SurfaceType26 26 :SurfaceType27 0 :SurfaceType28 0 :SurfaceType29 0 :SurfaceType30 0 :SurfaceType31 0 :SurfaceType32 0 :SurfaceType33 0 :SurfaceType34 0)))

(event EventSwingRight
  (|CallParentFunction)
  (Variables|Default|SetRight%20? true))

(event EventSwingLeft
  (|CallParentFunction)
  (Variables|Default|SetRight%20? false))

(fn ConstructionScript
  (|CallParentFunction)
  (Components|Activation|SetAutoActivate (CallFunction|GetIdlePSC) false))
