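; Cosmetic Blueprint for the Sparks microphone pickaxe, driving its swing and trail VFX.
; On equip it feeds the pawn mesh to the Niagara swing component; construction disables idle effect auto-activation.
; During montage section "Swing1", swings hide that component and, where cosmetics can run, attach a first-attack effect, activated after applying loadout variants.
; Swing state carries between trail notifies to set "IsR2" on the trail following a left swing, marking the third swing.
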(event FortWeapon|EventOnEquippedAndAttached
  (|CallParentFunction)
  (bind _asfort_pawn (Utilities|Casting|CastToFortPawn (Game|GetInstigator)))
  (Utilities|IsValid (Variables|/Script/Engine.Character|GetMesh _asfort_pawn)
    (:Is%20Valid
      (bind _asniagara_particle_system_component (Utilities|Casting|CastToNiagaraComponent (CallFunction|GetSwingVFXComponent))
        (:then
          (Niagara|SetNiagaraSkeletalMeshComponent _asniagara_particle_system_component "Pawn" (Variables|/Script/Engine.Character|GetMesh _asfort_pawn)))
        (:CastFailed
          (Development|PrintString "CAST FAILED - Niagara Component"))))))

(event EventAnimTrailsNotify
  (params (bActive :type Boolean))
  ;[1 1 1] Set the third swing
  (|CallParentFunction bActive)
  (Effects|Components|ParticleSystem|SetBoolParameter (CallFunction|GetAnimTrailVFXComponent) "IsR2" false)
  (if (== (Variables|Default|GetInt) 2)
    (Variables|Default|Setbool true)
    (Variables|Default|SetInt 0)
    (elif (Variables|Default|Getbool)
      (Effects|Components|ParticleSystem|SetBoolParameter (CallFunction|GetAnimTrailVFXComponent) "IsR2" true)
      (Variables|Default|Setbool false))))

(event EventSwingRight
  (bind _returnvalue (Game|GetInstigator))
  (bind _asfort_pawn (Utilities|Casting|CastToFortPawn _returnvalue))
  (bind _mesh (Variables|/Script/Engine.Character|GetMesh _asfort_pawn))
  (bind _returnvalue_2 (Components|SkeletalMesh|GetAnimInstance _mesh))
  (bind _returnvalue_3 (Animation|Montage|GetCurrentActiveMontage _returnvalue_2))
  (bind _returnvalue_4 (Animation|Montage|MontageGetCurrentSection _returnvalue_2 _returnvalue_3))
  (bind _returnvalue_5 (== _returnvalue_4 "Swing1"))
  ;[1 1 1] Set the third swing
  (|CallParentFunction)
  (Variables|Default|SetInt 1)
  (|CallParentFunction)
  (Networking|CanExecuteCosmeticEvents
    (:True
      (Utilities|FlowControl|Sequence
        (:then_0
          (if _returnvalue_5
            (bind _returnvalue_6 (Effects|SpawnFXSystemAttached :FXSystemAsset (Variables|Default|GetFXFirstAttack) :AttachToComponent (Variables|Visual|GetWeaponMesh) :AttachPointName "handle" :Scale "1.000000,1.000000,1.000000" :LocationType "KeepRelativeOffset" :bAutoDestroy true :PoolingMethod "None" :bAutoActivate false))
            (Class|FortCosmeticFunctionLibrary|ApplyVariantstoComponentbySlotTag (Utilities|Casting|CastToFortPlayerPawn _returnvalue) "(TagName=\"LoadoutSlot.Pickaxe\")" _returnvalue_6
              (:True
                (Components|Activation|Activate _returnvalue_6 true)))))
        (:then_1
          (Development|SetHiddeninGame (CallFunction|GetSwingVFXComponent) _returnvalue_5 false))))))

(event EventSwingLeft
  (bind _returnvalue (Game|GetInstigator))
  (bind _asfort_pawn (Utilities|Casting|CastToFortPawn _returnvalue))
  (bind _mesh (Variables|/Script/Engine.Character|GetMesh _asfort_pawn))
  (bind _returnvalue_2 (Components|SkeletalMesh|GetAnimInstance _mesh))
  (bind _returnvalue_3 (Animation|Montage|GetCurrentActiveMontage _returnvalue_2))
  (bind _returnvalue_4 (Animation|Montage|MontageGetCurrentSection _returnvalue_2 _returnvalue_3))
  (bind _returnvalue_5 (== _returnvalue_4 "Swing1"))
  ;[1 1 1] Set the third swing
  (|CallParentFunction)
  (Variables|Default|SetInt 2)
  (|CallParentFunction)
  (Networking|CanExecuteCosmeticEvents
    (:True
      (Utilities|FlowControl|Sequence
        (:then_0
          (if _returnvalue_5
            (bind _returnvalue_6 (Effects|SpawnFXSystemAttached :FXSystemAsset (Variables|Default|GetFXFirstAttack) :AttachToComponent (Variables|Visual|GetWeaponMesh) :AttachPointName "handle" :Scale "1.000000,1.000000,1.000000" :LocationType "KeepRelativeOffset" :bAutoDestroy true :PoolingMethod "None" :bAutoActivate false))
            (Class|FortCosmeticFunctionLibrary|ApplyVariantstoComponentbySlotTag (Utilities|Casting|CastToFortPlayerPawn _returnvalue) "(TagName=\"LoadoutSlot.Pickaxe\")" _returnvalue_6
              (:True
                (Components|Activation|Activate _returnvalue_6 true)))))
        (:then_1
          (Development|SetHiddeninGame (CallFunction|GetSwingVFXComponent) _returnvalue_5 false))))))

(fn ConstructionScript
  (|CallParentFunction)
  (Components|Activation|SetAutoActivate (CallFunction|GetIdlePSC) false))
